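# Physics asset for the Ballerina backpack cosmetic, set up for two dynamic wings on the character's back.
# Each wing is a capsule chain from Wing_01 through dyn_wing_02 to 08, with compound capsules on middle segments and twist-locked, swing-limited joints tuned per side.
# Capsules on pelvis, spine, arms and legs, joined by free-rotation constraints, approximate the wearer's body.
# An unconstrained floorCOL box adds a thin floor collider.
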PhysicsAsset /BRCosmetics/Accessories/FORT_Backpacks/Backpack_F_MED_Ballerina/Meshes/F_MED_Ballerina_Pack_Physics
bodies {
 Wing_01_L [{
   shapeType=Capsule
   length=8.8079
   radius=4.8023
   shapeName=Wing_01_L_capsule
   center=<-3.9224 0.5632 0.2571>
   rotation=<rot 0.1978 -96.5188 -86.847>
  }]
 Wing_01_R [{
   shapeType=Capsule
   length=10.0579
   radius=4.5523
   shapeName=Wing_01_R_capsule
   center=<-4.1052 0.6272 -0.0972>
   rotation=<rot -0.0554 -91.9279 -91.2454>
  }]
 calf_l [{shapeType=Capsule, length=31, radius=10.75, shapeName=calf_l_capsule, center=<-21.6338 0 0>, rotation=<rot 90 0 0>}]
 calf_r [{shapeType=Capsule, length=31, radius=10.75, shapeName=calf_r_capsule, center=<21.3497 0 0>, rotation=<rot 90 0 0>}]
 dyn_wing_02_l [{
   shapeType=Capsule
   length=9.2642
   radius=3.973
   shapeName=Wing_02_R_capsule
   center=<-0.0072 -0.4399 -0.0214>
   rotation=<rot 89.8735 -179.9985 176.4027>
  }]
 dyn_wing_02_r [{shapeType=Capsule, length=8.221, radius=3.9226, shapeName=Wing_02_R_capsule, center=<0.0862 0 -0.0047>, rotation=<rot 90 0 0>}]
 dyn_wing_03_l [
  {shapeType=Capsule, length=13.25, radius=3.901, shapeName=Wing_03_R_capsule, center=<4.7 -0.6837 0>, rotation=<rot 90 180 177.854>}
  {
   shapeType=Capsule
   length=13.75
   radius=4.285
   shapeName=Wing_03_L_capsule
   center=<5.5543 -0.896 -3.783>
   rotation=<rot 80.986 -0.2903 -2.3661>
  }
  {
   shapeType=Capsule
   length=15
   radius=3.5958
   shapeName=Wing_03_L_capsule
   center=<5.1308 0.2002 4.6879>
   rotation=<rot 81.0722 -179.3198 175.9112>
  }
 ]
 dyn_wing_03_r [
  {shapeType=Capsule, length=13.25, radius=3.2135, shapeName=Wing_03_R_capsule, center=<3.5236 0.4163 0>, rotation=<rot 90 180 174.5516>}
  {
   shapeType=Capsule
   length=15
   radius=3.7764
   shapeName=Wing_03_R_capsule
   center=<6.4153 -1.2281 3.767>
   rotation=<rot 80.9602 -179.9999 -178.813>
  }
  {
   shapeType=Capsule
   length=15
   radius=3.7533
   shapeName=Wing_03_R_capsule
   center=<6.8465 0.3602 -4.7119>
   rotation=<rot 81.004 0.0002 -5.3859>
  }
 ]
 dyn_wing_04_l [
  {shapeType=Capsule, length=13, radius=5.6268, shapeName=Wing_04_R_capsule, center=<8.784 -0.2982 0.1124>, rotation=<rot 90 179.6835 179.6835>}
  {shapeType=Capsule, length=13, radius=5, shapeName=Wing_04_L_capsule, center=<7.7079 0 -5.7531>, rotation=<rot 81.8991 0 0>}
  {shapeType=Capsule, length=15, radius=4.6819, shapeName=Wing_04_L_capsule, center=<8.7212 0 5.9533>, rotation=<rot 84.4982 -179.9999 -180>}
 ]
 dyn_wing_04_r [
  {shapeType=Capsule, length=11.5, radius=5.1344, shapeName=Wing_04_R_capsule, center=<8.0082 -0.4196 0.277>, rotation=<rot 90 0 0>}
  {shapeType=Capsule, length=11.5, radius=5.09, shapeName=Wing_04_R_capsule, center=<8.2928 0 5.9937>, rotation=<rot 82.6337 -180 180>}
  {shapeType=Capsule, length=15, radius=4.7511, shapeName=Wing_04_R_capsule, center=<8.52 0 -7.3298>, rotation=<rot 84.3716 0.0002 0.0001>}
 ]
 dyn_wing_05_l [
  {shapeType=Capsule, length=13, radius=6.6558, shapeName=Wing_05_R_capsule, center=<8.9462 -0.0211 -0.0476>, rotation=<rot 90 0 0>}
  {shapeType=Capsule, length=13.75, radius=5, shapeName=Wing_05_L_capsule, center=<8.254 0 -6.7428>, rotation=<rot 88.0619 -179.9999 -179.9999>}
 ]
 dyn_wing_05_r [
  {shapeType=Capsule, length=10.9376, radius=6.8904, shapeName=Wing_05_R_capsule, center=<8.2481 -0.0664 -0.0445>, rotation=<rot 90 0 0>}
  {shapeType=Capsule, length=10.5, radius=5, shapeName=Wing_05_R_capsule, center=<7.1843 0 6.7275>, rotation=<rot 88.1769 0.0001 0.0001>}
 ]
 dyn_wing_06_l [
  {shapeType=Capsule, length=12, radius=5.8485, shapeName=Wing_06_R_capsule, center=<5.5029 0 0.1348>, rotation=<rot 89.4037 -0.0003 -0.0003>}
  {shapeType=Capsule, length=15, radius=5, shapeName=Wing_06_L_capsule, center=<6.6831 0 -4.4983>, rotation=<rot 77.0274 -180 -180>}
 ]
 dyn_wing_06_r [
  {shapeType=Capsule, length=12, radius=6.1044, shapeName=Wing_06_R_capsule, center=<6.8352 -0.323 0.2138>, rotation=<rot 90 0 0>}
  {shapeType=Capsule, length=16.25, radius=5, shapeName=Wing_06_R_capsule, center=<6.1086 0 4.4532>, rotation=<rot 77.5921 0.0001 0>}
 ]
 dyn_wing_07_l [{shapeType=Capsule, length=6, radius=5, shapeName=Wing_07_R_capsule, center=<4.5381 0 -0.0998>, rotation=<rot 88.3494 0 -0.0001>}]
 dyn_wing_07_r [{shapeType=Capsule, length=6.5, radius=5, shapeName=Wing_07_R_capsule, center=<4.9551 0 0.1672>, rotation=<rot 87.0817 180 179.9999>}]
 dyn_wing_08_l [{
   shapeType=Capsule
   length=6.5
   radius=4.25
   shapeName=Wing_08_R_capsule
   center=<3.4813 0.0007 0.0297>
   rotation=<rot 89.442 -0.0001 -0.0001>
  }]
 dyn_wing_08_r [{shapeType=Capsule, length=6.5, radius=4.25, shapeName=Wing_08_R_capsule, center=<3.0848 -0.0002 0.2095>, rotation=<rot 90 0 0>}]
 floorCOL [{shapeType=Box, extentX=150, extentY=200, extentZ=1, shapeName=floorCOL_box}]
 lowerarm_l [{shapeType=Capsule, length=24.1997, radius=5.6176, shapeName=lowerarm_l_capsule, center=<14.9037 0 0>, rotation=<rot 90 0 0>}]
 lowerarm_r [{shapeType=Capsule, length=26.8581, radius=5.5686, shapeName=lowerarm_r_capsule, center=<-17.8835 -2.8944 0>, rotation=<rot 90 0 0>}]
 pelvis [{shapeType=Capsule, length=18.25, radius=18.5, shapeName=pelvis_capsule, center=<-9.9247 -1.7362 0>, rotation=<rot 0 -0.0001 0>}]
 spine_01 [{shapeType=Capsule, length=23.3193, radius=12.502, shapeName=spine_01_capsule, center=<0.0013 2.4144 -0.5593>, rotation=<rot 0.1281 0 0>}]
 spine_03 [{
   shapeType=Capsule
   length=24.8844
   radius=10.5658
   shapeName=spine_03_capsule
   center=<3.6771 2.2146 0.1099>
   rotation=<rot 1.8255 6.0342 0.3014>
  }]
 spine_04 [{shapeType=Capsule, length=4.6222, radius=9.0201, shapeName=spine_04_capsule, center=<10.737 4.3078 0>, rotation=<rot 90 -179.1048 173.9881>}]
 thigh_l [{shapeType=Capsule, length=25.922, radius=15, shapeName=thigh_l_capsule, center=<-17.4423 -1.6724 0>, rotation=<rot 90 180 -175>}]
 thigh_r [{shapeType=Capsule, length=25.243, radius=15, shapeName=thigh_r_capsule, center=<18.8343 2.0229 0>, rotation=<rot 90 179.5524 -175.4476>}]
 upperarm_l [{shapeType=Capsule, length=23.2581, radius=5.4182, shapeName=upperarm_l_capsule, center=<13.102 0 0>, rotation=<rot 90 0 0>}]
 upperarm_r [{
   shapeType=Capsule
   length=25.5491
   radius=5.4397
   shapeName=upperarm_r_capsule
   center=<-13.5448 -2.0853 0>
   rotation=<rot 90 -179.2838 -176.8839>
  }]
}
constraints [
 {
  bone1Name=spine_01
  bone2Name=pelvis
  swing1LimitDegrees=45
  swing1Motion=Free
  swing2LimitDegrees=45
  swing2Motion=Free
  twistLimitDegrees=45
  twistMotion=Free
 }
 {
  bone1Name=spine_03
  bone2Name=spine_01
  swing1LimitDegrees=45
  swing1Motion=Free
  swing2LimitDegrees=45
  swing2Motion=Free
  twistLimitDegrees=45
  twistMotion=Free
 }
 {
  bone1Name=spine_04
  bone2Name=spine_03
  swing1LimitDegrees=45
  swing1Motion=Free
  swing2LimitDegrees=45
  swing2Motion=Free
  twistLimitDegrees=45
  twistMotion=Free
 }
 {
  bone1Name=lowerarm_l
  bone2Name=upperarm_l
  swing1LimitDegrees=45
  swing1Motion=Free
  swing2LimitDegrees=45
  swing2Motion=Free
  twistLimitDegrees=45
  twistMotion=Free
 }
 {
  bone1Name=upperarm_l
  bone2Name=spine_04
  swing1LimitDegrees=45
  swing1Motion=Free
  swing2LimitDegrees=45
  swing2Motion=Free
  twistLimitDegrees=45
  twistMotion=Free
 }
 {
  bone1Name=upperarm_r
  bone2Name=spine_04
  swing1LimitDegrees=45
  swing1Motion=Free
  swing2LimitDegrees=45
  swing2Motion=Free
  twistLimitDegrees=45
  twistMotion=Free
 }
 {
  bone1Name=lowerarm_r
  bone2Name=upperarm_r
  swing1LimitDegrees=45
  swing1Motion=Free
  swing2LimitDegrees=45
  swing2Motion=Free
  twistLimitDegrees=45
  twistMotion=Free
 }
 {
  bone1Name=thigh_l
  bone2Name=pelvis
  swing1LimitDegrees=45
  swing1Motion=Free
  swing2LimitDegrees=45
  swing2Motion=Free
  twistLimitDegrees=45
  twistMotion=Free
 }
 {
  bone1Name=calf_l
  bone2Name=thigh_l
  swing1LimitDegrees=45
  swing1Motion=Free
  swing2LimitDegrees=45
  swing2Motion=Free
  twistLimitDegrees=45
  twistMotion=Free
 }
 {
  bone1Name=thigh_r
  bone2Name=pelvis
  swing1LimitDegrees=45
  swing1Motion=Free
  swing2LimitDegrees=45
  swing2Motion=Free
  twistLimitDegrees=45
  twistMotion=Free
 }
 {
  bone1Name=calf_r
  bone2Name=thigh_r
  swing1LimitDegrees=45
  swing1Motion=Free
  swing2LimitDegrees=45
  swing2Motion=Free
  twistLimitDegrees=45
  twistMotion=Free
 }
 {bone1Name=dyn_wing_02_l, bone2Name=wing_01_l, swing1LimitDegrees=17, swing2LimitDegrees=10, twistLimitDegrees=45, twistMotion=Locked}
 {bone1Name=dyn_wing_03_l, bone2Name=dyn_wing_02_l, swing1LimitDegrees=30, swing2LimitDegrees=17, twistLimitDegrees=45, twistMotion=Locked}
 {bone1Name=dyn_wing_04_l, bone2Name=dyn_wing_03_l, swing1LimitDegrees=30, swing2LimitDegrees=27, twistLimitDegrees=45, twistMotion=Locked}
 {bone1Name=dyn_wing_05_l, bone2Name=dyn_wing_04_l, swing1LimitDegrees=30, swing2LimitDegrees=27, twistLimitDegrees=45, twistMotion=Locked}
 {bone1Name=dyn_wing_06_l, bone2Name=dyn_wing_05_l, swing1LimitDegrees=30, swing2LimitDegrees=47, twistLimitDegrees=45, twistMotion=Locked}
 {bone1Name=dyn_wing_07_l, bone2Name=dyn_wing_06_l, swing1LimitDegrees=30, swing2LimitDegrees=45, twistLimitDegrees=45, twistMotion=Locked}
 {bone1Name=dyn_wing_08_l, bone2Name=dyn_wing_07_l, swing1LimitDegrees=30, swing2LimitDegrees=35, twistLimitDegrees=45, twistMotion=Locked}
 {bone1Name=dyn_wing_02_r, bone2Name=wing_01_r, swing1LimitDegrees=23, swing2LimitDegrees=13, twistLimitDegrees=45, twistMotion=Locked}
 {bone1Name=dyn_wing_03_r, bone2Name=dyn_wing_02_r, swing1LimitDegrees=30, swing2LimitDegrees=15, twistLimitDegrees=45, twistMotion=Locked}
 {bone1Name=dyn_wing_04_r, bone2Name=dyn_wing_03_r, swing1LimitDegrees=30, swing2LimitDegrees=25, twistLimitDegrees=45, twistMotion=Locked}
 {bone1Name=dyn_wing_05_r, bone2Name=dyn_wing_04_r, swing1LimitDegrees=30, swing2LimitDegrees=23, twistLimitDegrees=45, twistMotion=Locked}
 {bone1Name=dyn_wing_06_r, bone2Name=dyn_wing_05_r, swing1LimitDegrees=30, swing2LimitDegrees=32, twistLimitDegrees=45, twistMotion=Locked}
 {bone1Name=dyn_wing_07_r, bone2Name=dyn_wing_06_r, swing1LimitDegrees=30, swing2LimitDegrees=30, twistLimitDegrees=45, twistMotion=Locked}
 {bone1Name=dyn_wing_08_r, bone2Name=dyn_wing_07_r, swing1LimitDegrees=30, swing2LimitDegrees=30, twistLimitDegrees=45, twistMotion=Locked}
]
