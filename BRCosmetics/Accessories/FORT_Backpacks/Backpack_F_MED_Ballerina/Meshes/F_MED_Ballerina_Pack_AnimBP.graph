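# Animation Blueprint for the winged Ballerina backpack cosmetic, animating its wings as the player moves, crouches or skydives.
# State-based offsets feed a cached pose, then body and floorCOL bones take world-space transforms from variables.
# Three state-gated RigidBody sims, wing Trail controllers, walk offsets and a Control Rig follow in series.
# The EventGraph sets gravity, inertia scales and rig parameters per movement state, with low-LOD fallbacks when rigid bodies are off.
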AnimBlueprint /BRCosmetics/Accessories/FORT_Backpacks/Backpack_F_MED_Ballerina/Meshes/F_MED_Ballerina_Pack_AnimBP
target_skeleton=&/F_MED_Ballerina_Pack_Skeleton
graphs [
 AnimGraph {nodes [
   n1 = "Transform (Modify) Bone - Bone: body_root"(
    ComponentPose=Local To Component(LocalPose=Use cached pose 'InitPose'())
    Alpha=Get IsEmotingAlpha() -> (IsEmotingAlpha:float)
    Translation=Get BodyRootTran() -> (BodyRootTran:Vector)
    Rotation=Get BodyRootRot() -> (BodyRootRot:Rotator)
    Scale="1,1,1"
   ) {
    rotationMode=BMM_Replace
    rotationSpace=BCS_WorldSpace
    scaleSpace=BCS_BoneSpace
    translationMode=BMM_Replace
    translationSpace=BCS_WorldSpace
    alphaBoolBlend {blendOption=Linear}
    alphaScaleBiasClamp {inRange {max=0, min=-40}, outRange {max=0, min=1}}
    boneToModify {boneName=body_root}
   }
   n2 = "Transform (Modify) Bone - Bone: pelvis"(
    ComponentPose=@n1
    Alpha=1
    Translation=Get PelvisTran() -> (PelvisTran:Vector)
    Rotation=Get PelvisRot() -> (PelvisRot:Rotator)
    Scale="1,1,1"
   ) {
    rotationMode=BMM_Replace
    rotationSpace=BCS_WorldSpace
    scaleSpace=BCS_BoneSpace
    translationMode=BMM_Replace
    translationSpace=BCS_WorldSpace
    alphaBoolBlend {blendOption=Linear}
    alphaScaleBiasClamp {inRange {max=0, min=-40}, outRange {max=0, min=1}}
    boneToModify {boneName=pelvis}
   }
   n3 = "Transform (Modify) Bone - Bone: floorCOL"!(
    ComponentPose=@n2
    bAlphaBoolEnabled:bool=NOT Boolean(A=Get bIsDBNOCarried())
    Translation=Get RootTran() -> (RootTran:Vector)
    Rotation=Get RootRot() -> (RootRot:Rotator)
    Scale="1,1,1"
   ) -> (Pose) {
    alphaInputType=Bool
    rotationMode=BMM_Replace
    rotationSpace=BCS_WorldSpace
    scaleSpace=BCS_BoneSpace
    translationMode=BMM_Replace
    translationSpace=BCS_WorldSpace
    alphaBoolBlend {blendOption=Linear}
    alphaScaleBiasClamp {inRange {max=0, min=-40}, outRange {max=0, min=1}}
    boneToModify {boneName=floorCOL}
   }
   n4 = "Transform (Modify) Bone - Bone: thigh_l"(
    ComponentPose=@n3
    Alpha=1
    Translation=Get ThighLTran() -> (ThighLTran:Vector)
    Rotation=Get ThighLRot() -> (ThighLRot:Rotator)
    Scale="1,1,1"
   ) {
    rotationMode=BMM_Replace
    rotationSpace=BCS_WorldSpace
    scaleSpace=BCS_BoneSpace
    translationMode=BMM_Replace
    translationSpace=BCS_WorldSpace
    alphaBoolBlend {blendOption=Linear}
    alphaScaleBiasClamp {inRange {max=0, min=-40}, outRange {max=0, min=1}}
    boneToModify {boneName=thigh_l}
   }
   n5 = "Transform (Modify) Bone - Bone: thigh_r"(
    ComponentPose=@n4
    Alpha=1
    Translation=Get ThighRTran() -> (ThighRTran:Vector)
    Rotation=Get ThighRRot() -> (ThighRRot:Rotator)
    Scale="1,1,1"
   ) {
    rotationMode=BMM_Replace
    rotationSpace=BCS_WorldSpace
    scaleSpace=BCS_BoneSpace
    translationMode=BMM_Replace
    translationSpace=BCS_WorldSpace
    alphaBoolBlend {blendOption=Linear}
    alphaScaleBiasClamp {inRange {max=0, min=-40}, outRange {max=0, min=1}}
    boneToModify {boneName=thigh_r}
   }
   n6 = "Transform (Modify) Bone - Bone: calf_l"(ComponentPose=@n5, Alpha=1, Translation=Get CalfLTran() -> (CalfLTran:Vector), Rotation=Get CalfLRot() -> (CalfLRot:Rotator), Scale="1,1,1") {
    rotationMode=BMM_Replace
    rotationSpace=BCS_WorldSpace
    scaleSpace=BCS_BoneSpace
    translationMode=BMM_Replace
    translationSpace=BCS_WorldSpace
    alphaBoolBlend {blendOption=Linear}
    alphaScaleBiasClamp {inRange {max=0, min=-40}, outRange {max=0, min=1}}
    boneToModify {boneName=calf_l}
   }
   n7 = "Transform (Modify) Bone - Bone: calf_r"(ComponentPose=@n6, Alpha=1, Translation=Get CalfRTran() -> (CalfRTran:Vector), Rotation=Get CalfRRot() -> (CalfRRot:Rotator), Scale="1,1,1") {
    rotationMode=BMM_Replace
    rotationSpace=BCS_WorldSpace
    scaleSpace=BCS_BoneSpace
    translationMode=BMM_Replace
    translationSpace=BCS_WorldSpace
    alphaBoolBlend {blendOption=Linear}
    alphaScaleBiasClamp {inRange {max=0, min=-40}, outRange {max=0, min=1}}
    boneToModify {boneName=calf_r}
   }
   n8 = "Transform (Modify) Bone - Bone: spine_01"(
    ComponentPose=@n7
    Alpha=1
    Translation=Get Spine01Tran() -> (Spine01Tran:Vector)
    Rotation=Get Spine01Rot() -> (Spine01Rot:Rotator)
    Scale="1,1,1"
   ) {
    rotationMode=BMM_Replace
    rotationSpace=BCS_WorldSpace
    scaleSpace=BCS_BoneSpace
    translationMode=BMM_Replace
    translationSpace=BCS_WorldSpace
    alphaBoolBlend {blendOption=Linear}
    alphaScaleBiasClamp {inRange {max=0, min=-40}, outRange {max=0, min=1}}
    boneToModify {boneName=spine_01}
   }
   n9 = "Transform (Modify) Bone - Bone: spine_02"(
    ComponentPose=@n8
    Alpha=1
    Translation=Get Spine02Tran() -> (Spine02Tran:Vector)
    Rotation=Get Spine02Rot() -> (Spine02Rot:Rotator)
    Scale="1,1,1"
   ) {
    rotationMode=BMM_Replace
    rotationSpace=BCS_WorldSpace
    scaleSpace=BCS_BoneSpace
    translationMode=BMM_Replace
    translationSpace=BCS_WorldSpace
    alphaBoolBlend {blendOption=Linear}
    alphaScaleBiasClamp {inRange {max=0, min=-40}, outRange {max=0, min=1}}
    boneToModify {boneName=spine_02}
   }
   n10 = "Transform (Modify) Bone - Bone: spine_03"(
    ComponentPose=@n9
    Alpha=1
    Translation=Get Spine03Tran() -> (Spine03Tran:Vector)
    Rotation=Get Spine03Rot() -> (Spine03Rot:Rotator)
    Scale="1,1,1"
   ) {
    rotationMode=BMM_Replace
    rotationSpace=BCS_WorldSpace
    scaleSpace=BCS_BoneSpace
    translationMode=BMM_Replace
    translationSpace=BCS_WorldSpace
    alphaBoolBlend {blendOption=Linear}
    alphaScaleBiasClamp {inRange {max=0, min=-40}, outRange {max=0, min=1}}
    boneToModify {boneName=spine_03}
   }
   n11 = "Transform (Modify) Bone - Bone: spine_04"(
    ComponentPose=@n10
    Alpha=1
    Translation=Get Spine04Tran() -> (Spine04Tran:Vector)
    Rotation=Get Spine04Rot() -> (Spine04Rot:Rotator)
    Scale="1,1,1"
   ) {
    rotationMode=BMM_Replace
    rotationSpace=BCS_WorldSpace
    scaleSpace=BCS_BoneSpace
    translationMode=BMM_Replace
    translationSpace=BCS_WorldSpace
    alphaBoolBlend {blendOption=Linear}
    alphaScaleBiasClamp {inRange {max=0, min=-40}, outRange {max=0, min=1}}
    boneToModify {boneName=spine_04}
   }
   n12 = "Transform (Modify) Bone - Bone: spine_05"(
    ComponentPose=@n11
    Alpha=1
    Translation=Get Spine05Tran() -> (Spine05Tran:Vector)
    Rotation=Get Spine05Rot() -> (Spine05Rot:Rotator)
    Scale="1,1,1"
   ) {
    rotationMode=BMM_Replace
    rotationSpace=BCS_WorldSpace
    scaleSpace=BCS_BoneSpace
    translationMode=BMM_Replace
    translationSpace=BCS_WorldSpace
    alphaBoolBlend {blendOption=Linear}
    alphaScaleBiasClamp {inRange {max=0, min=-40}, outRange {max=0, min=1}}
    boneToModify {boneName=spine_05}
   }
   n13 = "Transform (Modify) Bone - Bone: clavicle_l"(
    ComponentPose=@n12
    Alpha=1
    Translation=Get ClavicleLTran() -> (ClavicleLTran:Vector)
    Rotation=Get ClavicleLRot() -> (ClavicleLRot:Rotator)
    Scale="1,1,1"
   ) {
    rotationMode=BMM_Replace
    rotationSpace=BCS_WorldSpace
    scaleSpace=BCS_BoneSpace
    translationMode=BMM_Replace
    translationSpace=BCS_WorldSpace
    alphaBoolBlend {blendOption=Linear}
    alphaScaleBiasClamp {inRange {max=0, min=-40}, outRange {max=0, min=1}}
    boneToModify {boneName=clavicle_l}
   }
   n14 = "Transform (Modify) Bone - Bone: clavicle_r"(
    ComponentPose=@n13
    Alpha=1
    Translation=Get ClavicleRTran() -> (ClavicleRTran:Vector)
    Rotation=Get ClavicleRRot() -> (ClavicleRRot:Rotator)
    Scale="1,1,1"
   ) {
    rotationMode=BMM_Replace
    rotationSpace=BCS_WorldSpace
    scaleSpace=BCS_BoneSpace
    translationMode=BMM_Replace
    translationSpace=BCS_WorldSpace
    alphaBoolBlend {blendOption=Linear}
    alphaScaleBiasClamp {inRange {max=0, min=-40}, outRange {max=0, min=1}}
    boneToModify {boneName=clavicle_r}
   }
   n15 = "Transform (Modify) Bone - Bone: upperarm_l"(
    ComponentPose=@n14
    Alpha=1
    Translation=Get UpperarmLTran() -> (UpperarmLTran:Vector)
    Rotation=Get UpperarmLRot() -> (UpperarmLRot:Rotator)
    Scale="1,1,1"
   ) {
    rotationMode=BMM_Replace
    rotationSpace=BCS_WorldSpace
    scaleSpace=BCS_BoneSpace
    translationMode=BMM_Replace
    translationSpace=BCS_WorldSpace
    alphaBoolBlend {blendOption=Linear}
    alphaScaleBiasClamp {inRange {max=0, min=-40}, outRange {max=0, min=1}}
    boneToModify {boneName=upperarm_l}
   }
   n16 = "Transform (Modify) Bone - Bone: upperarm_r"(
    ComponentPose=@n15
    Alpha=1
    Translation=Get UpperarmRTran() -> (UpperarmRTran:Vector)
    Rotation=Get UpperarmRRot() -> (UpperarmRRot:Rotator)
    Scale="1,1,1"
   ) {
    rotationMode=BMM_Replace
    rotationSpace=BCS_WorldSpace
    scaleSpace=BCS_BoneSpace
    translationMode=BMM_Replace
    translationSpace=BCS_WorldSpace
    alphaBoolBlend {blendOption=Linear}
    alphaScaleBiasClamp {inRange {max=0, min=-40}, outRange {max=0, min=1}}
    boneToModify {boneName=upperarm_r}
   }
   n17 = "Transform (Modify) Bone - Bone: lowerarm_l"(
    ComponentPose=@n16
    Alpha=1
    Translation=Get LowerarmLTran() -> (LowerarmLTran:Vector)
    Rotation=Get LowerarmLRot() -> (LowerarmLRot:Rotator)
    Scale="1,1,1"
   ) {
    rotationMode=BMM_Replace
    rotationSpace=BCS_WorldSpace
    scaleSpace=BCS_BoneSpace
    translationMode=BMM_Replace
    translationSpace=BCS_WorldSpace
    alphaBoolBlend {blendOption=Linear}
    alphaScaleBiasClamp {inRange {max=0, min=-40}, outRange {max=0, min=1}}
    boneToModify {boneName=lowerarm_l}
   }
   n18 = "Transform (Modify) Bone - Bone: lowerarm_r"(
    ComponentPose=@n17
    Alpha=1
    Translation=Get LowerarmRTran() -> (LowerarmRTran:Vector)
    Rotation=Get LowerarmRRot() -> (LowerarmRRot:Rotator)
    Scale="1,1,1"
   ) {
    rotationMode=BMM_Replace
    rotationSpace=BCS_WorldSpace
    scaleSpace=BCS_BoneSpace
    translationMode=BMM_Replace
    translationSpace=BCS_WorldSpace
    alphaBoolBlend {blendOption=Linear}
    alphaScaleBiasClamp {inRange {max=0, min=-40}, outRange {max=0, min=1}}
    boneToModify {boneName=lowerarm_r}
   }
   n19 = Get Acc Scale() -> (Acc Scale:Vector)
   n20 = Get Vel Scale() -> (Vel Scale:Vector)
   n21 = Get OverriteGravity() -> (OverriteGravity:Vector)
   n22 = RigidBody!(
    ComponentPose=@n18
    bAlphaBoolEnabled:bool=Get Is Not Crou/Sky() -> (Is Not Crou/Sky:bool)
    OverrideWorldGravity:Vector=@n21
    ExternalForce
    ComponentLinearAccScale:Vector=@n19
    ComponentLinearVelScale:Vector=@n20
   ) -> (Pose) {
    alphaInputType=Bool
    bClampLinearTranslationLimitToRefPose=true
    bOverrideWorldGravity=true
    overridePhysicsAsset=&/F_MED_Ballerina_Pack_Physics
    simulationSpace=BaseBoneSpace
    alphaBoolBlend {blendInTime=1, blendOption=Linear, blendOutTime=2}
    baseBoneRef {boneName=body_root}
    componentLinearAccScale=<0.8 0.8 0.8>
    componentLinearVelScale=<0.4 0.4 0.4>
    overrideWorldGravity {z=-50}
    simSpaceSettings {
     maxAngularAcceleration=100
     maxAngularVelocity=100
     maxLinearAcceleration=100
     maxLinearVelocity=100
     velocityScaleZ=0
     externalLinearDragV=<10 10 10>
    }
   }
   n23 = RigidBody(
    ComponentPose=@n22
    bAlphaBoolEnabled=Get Is FullySkyDiving() -> (Is FullySkyDiving:bool)
    OverrideWorldGravity=@n21
    ComponentLinearAccScale=@n19
    ComponentLinearVelScale=@n20
   ) {
    alphaInputType=Bool
    bClampLinearTranslationLimitToRefPose=true
    bOverrideWorldGravity=true
    overridePhysicsAsset=&F_MED_Ballerina_Pack_Physics
    simulationSpace=BaseBoneSpace
    alphaBoolBlend {blendInTime=1, blendOption=Linear, blendOutTime=2}
    baseBoneRef {boneName=backpack_root}
    componentLinearAccScale=<0.8 0.8 0.8>
    componentLinearVelScale=<0.4 0.4 0.4>
    overrideWorldGravity {z=-50}
    simSpaceSettings {
     maxAngularAcceleration=100
     maxAngularVelocity=100
     maxLinearAcceleration=100
     maxLinearVelocity=100
     velocityScaleZ=0
     externalLinearDragV=<10 10 10>
    }
   }
   n24 = RigidBody(
    ComponentPose=@n23
    bAlphaBoolEnabled=Get bIsCrouching()
    OverrideWorldGravity=@n21
    ComponentLinearAccScale=@n19
    ComponentLinearVelScale=@n20
   ) {
    alphaInputType=Bool
    bClampLinearTranslationLimitToRefPose=true
    bOverrideWorldGravity=true
    overridePhysicsAsset=&/F_MED_Ballerina_Pack_Physics_Crouch
    alphaBoolBlend {blendInTime=1, blendOption=Linear, blendOutTime=2}
    baseBoneRef {boneName=pack}
    componentAppliedLinearAccClamp=<2000 2000 2000>
    componentLinearAccScale=<0.8 0.8 0.8>
    componentLinearVelScale=<0.4 0.4 0.4>
    overrideWorldGravity {z=-50}
    simSpaceSettings {maxAngularAcceleration=100, maxAngularVelocity=100, maxLinearAcceleration=100, maxLinearVelocity=100}
   }
   n25 = Trail controller - Bone: dyn_wing_08_r!(ComponentPose=@n24, Alpha=0.2, RelaxationSpeedScale:float=1, FakeVelocity:Vector="1,1,1") -> (Pose) {
    bActorSpaceFakeVel=true
    bLimitRotation=true
    bShowTrailLocation=true
    bUsePlanarLimit=true
    chainLength=4
    debugLifeTime=0.1
    lODThreshold=1
    maxDeltaTime=0.0333
    stretchLimit=2
    alphaScaleBiasClamp {bInterpResult=true, interpSpeedDecreasing=25, interpSpeedIncreasing=25}
    baseJoint {boneName=backpack_root}
    relaxationSpeedScaleInputProcessor {bias=0.3, scale=0.4}
    rotationLimits [
     {limitMax=<0 180 0>, limitMin=<0 -180 0>}
     {limitMax=<0 180 0>, limitMin=<0 -180 0>}
     {limitMax=<0 176.9904 0>, limitMin=<0 -180 0>}
     {limitMax=<0 180 0>, limitMin=<0 -180 0>}
    ]
    rotationOffsets [<0 0 0>, <0 0 0>, <0 0 0>, <0 0 0>]
    trailBone {boneName=dyn_wing_08_r}
    trailRelaxationSpeed {editorCurveData {keys [<at 0 37.4961 RCIM_Linear>,<at 0.49 33.4961 RCIM_Linear>,<at 1 27.4961 RCIM_Linear>]}}
   }
   n26 = Trail controller - Bone: dyn_wing_08_l!(ComponentPose=@n25, Alpha=0.2, RelaxationSpeedScale:float=1, FakeVelocity:Vector="1,1,1") -> (Pose) {
    bActorSpaceFakeVel=true
    bLimitRotation=true
    bShowTrailLocation=true
    bUsePlanarLimit=true
    chainLength=4
    debugLifeTime=0.1
    lODThreshold=1
    maxDeltaTime=0.0333
    alphaScaleBiasClamp {bInterpResult=true, interpSpeedDecreasing=25, interpSpeedIncreasing=25}
    baseJoint {boneName=backpack_root}
    relaxationSpeedScaleInputProcessor {bias=0.3, scale=0.4}
    rotationLimits [
     {limitMax=<180 180 0>, limitMin=<0 -180 0>}
     {limitMax=<0 180 0>, limitMin=<0 -180 0>}
     {limitMax=<0 180 0>, limitMin=<0 -180 0>}
     {limitMax=<0 180 0>, limitMin=<0 -180 0>}
    ]
    rotationOffsets [<0 0 0>, <0 0 0>, <0 0 0>, <0 0 0>]
    trailBone {boneName=dyn_wing_08_l}
    trailRelaxationSpeed {editorCurveData {keys [<at 0 37.1289 RCIM_Linear>,<at 0.4574 33.1289 RCIM_Linear>,<at 1 27.1289 RCIM_Linear>]}}
   }
   n27 = Get Walk() -> (Walk:bool)
   n28 = "Transform (Modify) Bone - Bone: wing_01_r"!(ComponentPose=@n26, bAlphaBoolEnabled:bool=@n27, Translation="0,0,0", Rotation="0,-1,0", Scale="1,1,1") -> (Pose) {
    alphaInputType=Bool
    rotationMode=BMM_Additive
    rotationSpace=BCS_BoneSpace
    translationSpace=BCS_BoneSpace
    alphaBoolBlend {blendOption=Linear}
    boneToModify {boneName=wing_01_r}
    rotation {yaw=-1}
   }
   n29 = "Transform (Modify) Bone - Bone: wing_01_l"!(ComponentPose=@n28, bAlphaBoolEnabled:bool=@n27, Translation="0,0,0", Rotation="0,-0.4,0", Scale="1,1,1") -> (Pose) {
    alphaInputType=Bool
    rotationMode=BMM_Additive
    rotationSpace=BCS_BoneSpace
    translationSpace=BCS_BoneSpace
    alphaBoolBlend {blendOption=Linear}
    boneToModify {boneName=wing_01_l}
    rotation {yaw=-0.4}
   }
   n30 = Get Alpha Lod Walk() -> (Alpha Lod Walk:double)
   n31 = "Transform (Modify) Bone - Bone: dyn_wing_03_l"(ComponentPose=@n29, Alpha=@n30, Translation="0,0,0", Rotation="0,-20,0", Scale="1,1,1") {
    rotationMode=BMM_Additive
    rotationSpace=BCS_BoneSpace
    translationSpace=BCS_BoneSpace
    alphaBoolBlend {blendOption=Linear}
    boneToModify {boneName=dyn_wing_03_l}
    rotation {yaw=-20}
   }
   n32 = "Transform (Modify) Bone - Bone: dyn_wing_03_r"(ComponentPose=@n31, Alpha=@n30, Translation="0,0,0", Rotation="0,-20,0", Scale="1,1,1") {
    rotationMode=BMM_Additive
    rotationSpace=BCS_BoneSpace
    translationSpace=BCS_BoneSpace
    alphaBoolBlend {blendOption=Linear}
    boneToModify {boneName=dyn_wing_03_r}
    rotation {yaw=-20}
   }
   n33 = "Transform (Modify) Bone - Bone: wing_01_l"!(ComponentPose=@n32, Alpha:float=@n30, Translation="0,0,0", Rotation="0,-10,0", Scale="1,1,1") -> (Pose) {
    rotationMode=BMM_Additive
    rotationSpace=BCS_BoneSpace
    translationSpace=BCS_BoneSpace
    alphaBoolBlend {blendOption=Linear}
    boneToModify {boneName=wing_01_l}
    rotation {yaw=-10}
   }
   n34 = "Transform (Modify) Bone - Bone: wing_01_r"!(ComponentPose=@n33, Alpha:float=@n30, Translation="0,0,0", Rotation="0,-10,0", Scale="1,1,1") -> (Pose) {
    rotationMode=BMM_Additive
    rotationSpace=BCS_BoneSpace
    translationSpace=BCS_BoneSpace
    alphaBoolBlend {blendOption=Linear}
    boneToModify {boneName=wing_01_r}
    rotation {yaw=-10}
   }
   n35 = Control Rig!(
    Source=Component To Local(ComponentPose=@n34)
    Alpha
    Speed:double=Get CtrRig Speed() -> (CtrRig Speed:double)
    Amplitude:double=Get CtrRig Amp() -> (CtrRig Amp:double)
    Frequency:double=Get CtrRig Freq() -> (CtrRig Freq:double)
    PenGravity:Vector=Get OverriteGravity()
    Is Moving:bool=@n27
    Is Fully SkyDiving:bool=Get Is FullySkyDiving()
   ) -> (Pose) {
    alphaBoolBlend {blendOption=Linear}
    controlRigAssetReference {blueprintRigClass=&/F_MED_Ballerina_Pack_CtrlRig.F_MED_Ballerina_Pack_CtrlRig_C}
   }
   Output Pose(Result=@n35) {layerGroup=None}
   n36 = "Transform (Modify) Bone - Bone: backpack_root"!(
    ComponentPose=Local To Component(LocalPose=Input Pose())
    bAlphaBoolEnabled:bool=True
    Translation="0,0.35,-3.853"
    Rotation="-0.000,0.000,2.112"
    Scale="1,1,1"
   ) -> (Pose) {
    alphaInputType=Bool
    rotationMode=BMM_Additive
    rotationSpace=BCS_BoneSpace
    scaleSpace=BCS_BoneSpace
    translationMode=BMM_Additive
    translationSpace=BCS_BoneSpace
    alphaBoolBlend {blendInTime=0.1, blendOption=Linear, blendOutTime=-1}
    alphaScaleBiasClamp {bInterpResult=true}
    boneToModify {boneName=backpack_root}
    rotation {roll=2.112}
    translation {y=0.35, z=-3.853}
   }
   n37 = "Transform (Modify) Bone - Bone: backpack_root"(ComponentPose=@n36, bAlphaBoolEnabled=Get Is FullySkyDiving(), Translation="0,-0.8,0", Rotation="0.000,0.000,-15.000", Scale="1,1,1") {
    alphaInputType=Bool
    rotationMode=BMM_Additive
    rotationSpace=BCS_BoneSpace
    scaleSpace=BCS_BoneSpace
    translationMode=BMM_Additive
    translationSpace=BCS_BoneSpace
    alphaBoolBlend {blendInTime=0.1, blendOption=Linear, blendOutTime=-1}
    alphaScaleBiasClamp {bInterpResult=true}
    boneToModify {boneName=backpack_root}
    rotation {roll=-15}
    translation {y=-0.8}
   }
   n38 = "Transform (Modify) Bone - Bone: backpack_root"(ComponentPose=@n37, bAlphaBoolEnabled=Get Walk(), Translation="0.000,-1.118,0.000", Rotation="0,0,0", Scale="1,1,1") {
    alphaInputType=Bool
    translationMode=BMM_Additive
    translationSpace=BCS_BoneSpace
    alphaBoolBlend {blendOption=Linear}
    boneToModify {boneName=backpack_root}
    translation {y=-1.118}
   }
   n39 = "Transform (Modify) Bone - Bone: backpack_root"(ComponentPose=@n38, bAlphaBoolEnabled=Get bIsSkydiveDiveMode(), Translation="0,0,0", Rotation="0,0,0", Scale="1,1,1") {
    alphaInputType=Bool
    rotationMode=BMM_Additive
    rotationSpace=BCS_BoneSpace
    alphaBoolBlend {blendOption=Linear}
    boneToModify {boneName=backpack_root}
   }
   n40 = Get bIsInFrontEnd()
   n41 = "Transform (Modify) Bone - Bone: backpack_root"(ComponentPose=@n39, bAlphaBoolEnabled=@n40, Translation="-0.000,-1.039,2.697", Rotation="0,0,0", Scale="1,1,1") {
    alphaInputType=Bool
    rotationSpace=BCS_BoneSpace
    translationMode=BMM_Additive
    translationSpace=BCS_BoneSpace
    alphaBoolBlend {blendOption=Linear}
    boneToModify {boneName=backpack_root}
    translation {y=-1.039, z=2.697}
   }
   n42 = "Transform (Modify) Bone - Bone: wing_01_r"!(ComponentPose=@n41, bAlphaBoolEnabled:bool=@n40, Translation="-0.000,-0.829,0.929", Rotation="0.000,0.009,10.581", Scale="1,1,1") -> (Pose) {
    alphaInputType=Bool
    rotationMode=BMM_Additive
    rotationSpace=BCS_BoneSpace
    translationMode=BMM_Additive
    translationSpace=BCS_BoneSpace
    alphaBoolBlend {blendOption=Linear}
    boneToModify {boneName=wing_01_r}
    rotation=<rot 0 0.0091 10.581>
    translation {y=-0.829, z=0.929}
   }
   n43 = "Transform (Modify) Bone - Bone: wing_01_l"!(ComponentPose=@n42, bAlphaBoolEnabled:bool=@n40, Translation="0,-0.829,-0.929", Rotation="0,0,-10.581", Scale="1,1,1") -> (Pose) {
    alphaInputType=Bool
    rotationMode=BMM_Additive
    rotationSpace=BCS_BoneSpace
    translationMode=BMM_Additive
    translationSpace=BCS_BoneSpace
    alphaBoolBlend {blendOption=Linear}
    boneToModify {boneName=wing_01_l}
    rotation {roll=-10.581}
    translation {y=-0.829, z=-0.929}
   }
   n44 = "Transform (Modify) Bone - Bone: backpack_root"(
    ComponentPose=@n43
    bAlphaBoolEnabled=Get Is Low LOD SkyDive() -> (Is Low LOD SkyDive:bool)
    Translation="0.000,-5.045,3.392"
    Rotation="-0.000,0.000,31.999"
    Scale="1,1,1"
   ) {
    alphaInputType=Bool
    rotationMode=BMM_Additive
    rotationSpace=BCS_BoneSpace
    translationMode=BMM_Additive
    alphaBoolBlend {blendOption=Linear}
    boneToModify {boneName=backpack_root}
    rotation {roll=31.999}
    translation {y=-5.045, z=3.392}
   }
   n45 = "Transform (Modify) Bone - Bone: floorCOL"(ComponentPose=@n44, bAlphaBoolEnabled=Get bIsDBNOCarried(), Translation="0.000,0.000,67.459", Rotation="0,0,0", Scale="1,1,1") {
    alphaInputType=Bool
    translationMode=BMM_Additive
    alphaBoolBlend {blendOption=Linear}
    boneToModify {boneName=floorCOL}
    translation {z=67.459}
   }
   n46 = Get is Low LOD() -> (is Low LOD:bool)
   n47 = "Transform (Modify) Bone - Bone: wing_01_l"(ComponentPose=@n45, bAlphaBoolEnabled=@n46, Translation="0,0,0", Rotation="0,0,7", Scale="1,1,1") {alphaInputType=Bool, rotationMode=BMM_Additive, alphaBoolBlend {blendOption=Linear}, boneToModify {boneName=wing_01_l}, rotation {roll=7}}
   n48 = "Transform (Modify) Bone - Bone: wing_01_r"(ComponentPose=@n47, bAlphaBoolEnabled=@n46, Translation="0,0,0", Rotation="0,0,7", Scale="1,1,1") {alphaInputType=Bool, rotationMode=BMM_Additive, alphaBoolBlend {blendOption=Linear}, boneToModify {boneName=wing_01_r}, rotation {roll=7}}
   Save cached pose 'InitPose'(Pose=Component To Local(ComponentPose=@n48)) {cacheName=InitPose}
  ]}
 EventGraph {nodes [
   n1 = Event BlueprintUpdateAnimation()
   > n2 = Sequence()
   > n3 = Branch(Condition=Get is LOD Walk() -> (is LOD Walk:bool))
   > Set Alpha Lod Walk(Alpha Lod Walk:double=FInterpTo(Current=Get Alpha Lod Walk(), Target=1, DeltaTime=@n1.DeltaTimeX, InterpSpeed=5)) -> (Output_Get:double)
   Set Alpha Lod Walk(execute=@n3.else, Alpha Lod Walk=FInterpTo(Current=Get Alpha Lod Walk(), Target=0, DeltaTime=@n1.DeltaTimeX, InterpSpeed=5))
   n4 = Get Is FullySkyDiving()
   n5 = NOT Boolean(A=Get bIsCrouching())
   n6 = Get bIsMoving2D()
   Set Walk(execute=@n2.then_1, Walk:bool=AND Boolean(A=AND Boolean(A=@n6, B=@n5), B=NOT Boolean(A=@n4))) -> (Output_Get:bool)
   > n7 = Set Is FullySkyDiving(Is FullySkyDiving:bool=OR Boolean(A=Get bIsSkydiveFloating(), B=Get bIsSkydiveDiveMode())) -> (Output_Get:bool)
   > Set Is Not Crou/Sky(Is Not Crou/Sky:bool=NOR Boolean!(A:bool=@n7, B:bool=Get bIsCrouching()) -> (ReturnValue:bool)) -> (Output_Get:bool)
   > n8 = Sequence!(execute) -> (then_0, then_1, then_2:Exec)
   > n9 = Gravity Override State Function DEPRECATED(
    CharacterPartAnimInstance=Self-Reference()
    GameTimeInSeconds=GetGameTimeInSeconds()
    GravityOverrideParams=Get Gravity Override Params()
    DebugJointName=backpack_root
   )
   > n10 = Sequence!(execute) -> (then_0, then_1)
   > Set OverriteGravity(OverriteGravity:Vector=vector + vector(A=@n9, B=Get CrouchGravity() -> (CrouchGravity:Vector))) -> (Output_Get:Vector)
   n11 = Branch(execute=@n10.then_1, Condition=Get bIsCrouching())
   > Set CrouchGravity(CrouchGravity:Vector="0,0,-700") -> (Output_Get:Vector)
   Set CrouchGravity(execute=@n11.else, CrouchGravity="0, 0, 0")
   Set is Low LOD(
    execute=@n8.then_1
    is Low LOD:bool="Equal (Boolean)"(A:bool=Get bRigidBodyEnabled(), B:bool) -> (ReturnValue:bool=~) {is=title K2Node_PromotableOperator}
   ) -> (Output_Get:bool)
   n12 = "Equal (Boolean)"(A=Get bRigidBodyEnabled()) -> (ReturnValue=~)
   > n13 = Branch(Condition=NOT Boolean(A=@n12))
   > Set Is Low LOD SkyDive(Is Low LOD SkyDive:bool=false) -> (Output_Get:bool)
   > n14 = Branch(Condition=Get Is FullySkyDiving())
   > n15 = Sequence!(execute) -> (then_0, then_1, then_2:Exec, then_3:Exec, then_4:Exec)
   > Set CtrRig Speed(CtrRig Speed:double=20) -> (Output_Get:double)
   > Branch(Condition=Get bIsSkydiveDiveMode())
   > n16 = Sequence!(execute) -> (then_0, then_1, then_2)
   > Set CtrRig Speed(CtrRig Speed=35)
   Set CtrRig Amp(execute=@n16.then_1, CtrRig Amp:double=1.7) -> (Output_Get:double)
   Set CtrRig Freq(execute=@n16.then_2, CtrRig Freq:double=15) -> (Output_Get:double)
   Set CtrRig Amp(execute=@n15.then_1, CtrRig Amp=1.5)
   Set CtrRig Freq(execute=@n15.then_2, CtrRig Freq=15)
   Set Acc Scale(execute=@n15.then_3, Acc Scale:Vector="1,1,1") -> (Output_Get:Vector)
   Set Vel Scale(execute=@n15.then_4, Vel Scale:Vector="1,1,1") -> (Output_Get:Vector)
   n17 = Sequence!(execute=@n14.else) -> (then_0, then_1)
   > Branch(Condition=Get bIsCrouching())
   > n18 = Sequence()
   > Set Acc Scale(Acc Scale="0.25,0.25,0.25")
   Set Vel Scale(execute=@n18.then_1, Vel Scale="0.1,0.1,0.1")
   n19 = Branch(execute=@n17.then_1, Condition=Get bIsMoving2D())
   n20 = OR Boolean(A=Get bIsJumping(), B=Get bIsFalling())
   > n21 = Branch(Condition=@n20)
   > n22 = Sequence()
   > Set Acc Scale(Acc Scale="0.3,0.3,0.3")
   Set Vel Scale(execute=@n22.then_1, Vel Scale="0.2,0.2,0.2")
   n23 = Branch(execute=@n21.else, Condition=Get bIsSurfaceSwimming())
   > n24 = Sequence()
   > Set Acc Scale(Acc Scale="0.3,0.3,0.3")
   Set Vel Scale(execute=@n24.then_1, Vel Scale="0.2,0.2,0.2")
   n25 = Sequence(execute=@n23.else)
   > Set Acc Scale(Acc Scale="1,1,1")
   Set Vel Scale(execute=@n25.then_1, Vel Scale="0.2,0.2,0.2")
   n26 = Sequence(execute=@n19.else)
   > n27 = Branch(Condition=Get bIsDBNOCarried())
   > n28 = Sequence!(execute) -> (then_0, then_1, then_2:Exec)
   > Set CtrRig Speed(CtrRig Speed=0)
   Set CtrRig Amp(execute=@n28.then_1, CtrRig Amp=0.5)
   Set CtrRig Freq(execute=@n28.then_2, CtrRig Freq=100)
   n29 = Sequence(execute=@n27.else)
   > Set CtrRig Speed(CtrRig Speed=1)
   Set CtrRig Amp(execute=@n29.then_1, CtrRig Amp=0.5)
   Set CtrRig Freq(execute=@n29.then_2, CtrRig Freq=10)
   n30 = Branch(execute=@n26.then_1, Condition=@n20)
   > n31 = Sequence!(execute) -> (then_0, then_1)
   > Set Acc Scale(Acc Scale="0.5,0.5,0.5")
   Set Vel Scale(execute=@n31.then_1, Vel Scale="0.2,0.2,0.2")
   n32 = Sequence(execute=@n30.else)
   > Set Acc Scale(Acc Scale="0.3,0.3,0.3")
   Set Vel Scale(execute=@n32.then_1, Vel Scale="0.2,0.2,0.2")
   n33 = Sequence!(execute=@n13.else) -> (then_0, then_1, then_2:Exec, then_3:Exec)
   > Set CtrRig Speed(CtrRig Speed=1)
   Set CtrRig Amp(execute=@n33.then_1, CtrRig Amp=0.05)
   Set CtrRig Freq(execute=@n33.then_2, CtrRig Freq=40)
   Set Is Low LOD SkyDive(execute=@n33.then_3, Is Low LOD SkyDive="Equal (Boolean)"(A=@n12, B=Get Is FullySkyDiving()) -> (ReturnValue=~))
   n34 = Branch(execute=@n8.then_2, Condition=Get bIsInFrontEnd())
   Set is LOD Walk(execute=@n34.else, is LOD Walk:bool=AND Boolean(A=Get Walk(), B=Get Is Not Crou/Sky())) -> (Output_Get:bool)
   TryGetPawnOwner()
  ]}
]
variables [
 Acc Scale {type=Vector}
 Alpha Lod Walk {type=number}
 CrouchGravity {type=Vector}
 CtrRig Amp {type=number}
 CtrRig Freq {type=number}
 CtrRig Speed {type=number}
 Gravity Override Params {
  type=GravityOverrideParamsStruct
  default {
   emoteOrMelee {jointName=Wing_01_R, gravityOverride=<0 0 -400>}
   front_End {jointName=Wing_01_R, gravityOverride=<0 0 -602>}
   onGround_Falling {jointName=Wing_01_R, gravityOverride=<0 0 -700>}
   onGround_Moving {jointName=Wing_01_R, gravityOverride=<0 0 -500>, windAmplitude=<100 100 1300>, windFrequency=<5 5 20>, windOffset=<10 0 450>}
   onGround_Standing {jointName=Wing_01_R, gravityOverride=<0 0 -350>}
   riding_Moving {jointName=Wing_01_R}
   riding_Moving_Targeting {jointName=Wing_01_R}
   riding_Standing {jointName=Wing_01_R}
   skydive_DivingDown {jointName=backpack_root, gravityOverride=<0 0 1700>, windAmplitude=<0 1500 0>, windFrequency=<0 50 0>, windOffset=<0 3000 0>}
   skydive_FloatingDown {jointName=backpack_root, gravityOverride=<0 0 1300>, windAmplitude=<0 1500 0>, windFrequency=<0 50 0>, windOffset=<0 -1500 0>}
   skydive_GoingUp {jointName=Wing_01_R, gravityOverride=<0 0 1300>, windAmplitude=<0 1500 0>, windFrequency=<0 50 0>, windOffset=<0 1300 0>}
   skydive_Parachuting {jointName=Wing_01_R, gravityOverride=<0 -350 -400>}
   skydive_Surfing {jointName=Wing_01_R, gravityOverride=<500 -300 400>, windAmplitude=<0 0 0>, windFrequency=<0 0 0>}
   swimming {jointName=Wing_01_R, gravityOverride=<0 -1500 -1500>, windOffset=<0 300 0>}
   tactical_Sprint {jointName=Wing_01_R}
   ziplining {jointName=Wing_01_R}
  }
 }
 Is FullySkyDiving {type=boolean}
 is LOD Walk {type=boolean}
 is Low LOD {type=boolean}
 Is Low LOD SkyDive {type=boolean}
 Is Not Crou/Sky {type=boolean}
 OverriteGravity {type=Vector}
 Vel Scale {type=Vector}
 Walk {type=boolean}
]
